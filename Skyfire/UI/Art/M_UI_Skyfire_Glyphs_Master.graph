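# Translucent UI master material for Skyfire glyphs, picking one glyph from a texture atlas with FlipBook via a Glyph index and the row and column counts.
# The red channel, shaped by Subtract and Edge, forms a mask that blends Falloff_Color into Color.
# That mask plus the green channel forms the alpha, weighted by Glow_Amount and Final_Alpha.
# A horizontal Swipe wipe with Hardness and switchable direction multiplies that alpha.
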Material /Skyfire/UI/Art/M_UI_Skyfire_Glyphs_Master
graph {
 properties {blendMode=BLEND_Translucent, materialDomain=MD_UI, pixelDepthOffsetMode=PDOM_Legacy, usedShadingModels=Default Lit}
 nodes [
  n1 = ScalarParameter() {defaultValue=8, parameterName=Glyph}
  n2 = ScalarParameter() {defaultValue=8, parameterName=Rows}
  n3 = ScalarParameter() {defaultValue=2, parameterName=Columns}
  n4 = Add(A=Multiply(A=@n2, B=@n3)) {constB=-1}
  n5 = TextureSample(UVs=MaterialFunctionCall("Animation  Phase (0-1)"=Divide(A=@n1, B=@n4), Number of Rows=@n2, Number of Columns=@n3) {materialFunction=&FlipBook}.UVs) {samplerType=SAMPLERTYPE_LinearColor, texture=&/T_UI_Nevada_Glyphs_512x}
  n6 = Subtract(A=@n5.R, B=ScalarParameter() {parameterName=Subtract})
  n7 = Saturate(None=Multiply(A=Saturate(None=@n6), B=ScalarParameter() {defaultValue=1.125, parameterName=Edge}))
  LinearInterpolate(
   A=VectorParameter() {parameterName=Falloff_Color, defaultValue=<rgba 1 0 0 1>}.RGB
   B=VectorParameter() {parameterName=Color, defaultValue=<rgba 0 1 0 1>}.RGB
   Alpha=@n7
  )
  n8 = ScalarParameter() {defaultValue=0.5, parameterName=Glow_Amount}
  n9 = Add(A=@n7, B=@n5.G)
  n10 = Saturate(None=DotProduct(A=LinearInterpolate(A=@n8, Alpha=@n7), B=Saturate(None=@n9)))
  n11 = ScalarParameter() {defaultValue=1, parameterName=Final_Alpha}
  n12 = ComponentMask(None=TextureCoordinate()) {r=true}
  n13 = StaticSwitchParameter(True=OneMinus(None=@n12), False=@n12) {parameterName=FlipDirection}
  n14 = ScalarParameter() {defaultValue=-1, parameterName=Swipe}
  Multiply(
   A=Saturate(None=Multiply(A=@n10, B=@n11))
   B=Multiply(A=Subtract(A=@n13, B=@n14), B=ScalarParameter() {defaultValue=1, parameterName=Hardness}) {constB=5}
  )
 ]
}
