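# UI material for the Skyfire star tier indicator, drawing a tintable star with alpha-composite blending.
# One star texture packs masks by channel: R for outline, G for glow, B for fill.
# Chained lerps layer glow color, then fill scaled by its alpha, then outline on top.
# A separate sum adds scaled fill to the raw glow mask; glow alpha is unused despite its description.
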Material /Skyfire/UI/Art/M_UI_Skyfire_StarTierIndicator
graph {
 properties {blendMode=BLEND_AlphaComposite, materialDomain=MD_UI, pixelDepthOffsetMode=PDOM_Legacy, usedShadingModels=Default Lit}
 nodes [
  n1 = VectorParameter() {desc="fill color of interior of star (allows alpha)", parameterName=ColorFill, defaultValue=<rgba 1 0.0156 0.8718 1>}
  n2 = TextureSampleParameter2D() {parameterName=Texture, texture=&/T-Skyfire-Star-512}
  n3 = Multiply(A=@n1.A, B=@n2.B)
  Add(A=@n3, B=@n2.G)
  n4 = VectorParameter() {
   desc="blur glow color on exterior and interior of star - won't draw above the fill (alpha included)"
   parameterName=ColorGlow
   defaultValue=<rgba 1 0.4311 0 0>
  }
  n5 = VectorParameter() {desc="outline color on edge of star (no alpha)", parameterName=ColorOutline, defaultValue=<rgba 0.1593 1 0.0677 1>}
  LinearInterpolate(A=LinearInterpolate(A=LinearInterpolate(B=@n4.RGB, Alpha=@n2.G), B=@n1.RGB, Alpha=@n3), B=@n5.RGB, Alpha=@n2.R)
 ]
}
